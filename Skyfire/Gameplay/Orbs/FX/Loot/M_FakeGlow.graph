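# Translucent unlit material for a soft fake glow sprite, filed with the loot orb effects.
# A Sprite function handles the quad using a particle size parameter, and color is scaled by time-of-day values.
# A radial UV SphereMask with adjustable falloff is multiplied by color alpha, near-camera fade and depth fade.
# Depth fade uses the plain distance parameter, bypassing the dynamic-parameter multiplier, and ParticleColor feeds no other node.
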Material /Skyfire/Gameplay/Orbs/FX/Loot/M_FakeGlow
graph {
 properties {
  blendMode=BLEND_Translucent
  pixelDepthOffsetMode=PDOM_Legacy
  previewMesh=/Skyfire/Gameplay/Orbs/FX/Time/SM_SpriteSingular
  shadingModel=MSM_Unlit
  usedShadingModels=Unlit
 }
 nodes [
  n1 = VectorParameter() {parameterName=Particle size, channelNames {a="", b=Master, g=Y, r=X}, defaultValue=<rgba 50 50 1 0>}
  MaterialFunctionCall!(
   Center Location=WorldPosition().XYZ
   0-1 UVs
   XY scale=Multiply(A=ComponentMask(None=@n1) {g=true, r=true}, B=@n1.Master)
   OverridePivot
   "Normals (Optional)"
  ) {materialFunction=&Sprite}
  n2 = ScalarParameter() {defaultValue=100, parameterName=DepthFadeDist}
  Multiply(A=DynamicParameter() {defaultValue=<rgba 3 1 1 1>, paramNames [Power, DepthFadeDistMult, Param3, Param4]}.DepthFadeDistMult, B=@n2)
  n3 = VectorParameter() {parameterName=Color, defaultValue=<rgba 1 1 1 1>}
  Multiply(
   A=@n3.RGB
   B=MaterialFunctionCall(
    Value_Morning=ScalarParameter() {defaultValue=1, parameterName="Value_Morning (S)"}
    Value_Day=ScalarParameter() {defaultValue=1, parameterName="Value_Day (S)"}
    Value_Evening=ScalarParameter() {defaultValue=1, parameterName="Value_Evening (S)"}
    Value_Night=ScalarParameter() {defaultValue=1, parameterName="Value_Night (S)"}
   ) {materialFunction=&Scalar_Time_of_Day}.Result
  )
  n4 = TextureCoordinate()
  n5 = Constant() {r=0.5}
  n6 = ScalarParameter() {defaultValue=0.5, parameterName=Radius}
  n7 = ScalarParameter() {parameterName=Hardness}
  n8 = ScalarParameter() {defaultValue=1, parameterName=Power}
  n9 = MaterialFunctionCall(Distance=ScalarParameter() {defaultValue=100, parameterName=NearCameraFadeDist}) {materialFunction=&Near_Camera_Fade}
  Multiply(
   A=Power(
    Base=SphereMask(A=@n4, B=@n5, Radius=@n6, Hardness=@n7) {hardnessPercent=0}
    Exp=Clamp(None=@n8) {clampMode=CMODE_ClampMin, maxDefault=100, minDefault=1}
   ) {constExponent=3}
   B=Multiply(A=@n3.A, B=DepthFade(Opacity=@n9.Result, FadeDistance=@n2))
  )
  ParticleColor()
 ]
}
