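# Unlit translucent material for a simple procedural beam in orb loot FX, enabled for Niagara sprites, ribbons, mesh particles and beam trails.
# A generated band shapes the sides, while two opposed vertical linear gradients fade top and bottom.
# Each gradient has exponent, multiply, offset and contrast controls packed into one vector parameter.
# The three masks are multiplied, boosted and scaled by particle alpha, with no texture samples.
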Material /Skyfire/Gameplay/Orbs/FX/Loot/M_SimpleBeamForSpriteBasedLine
graph {
 properties {
  bUsedWithBeamTrails=true
  bUsedWithMeshParticles=true
  bUsedWithNiagaraMeshParticles=true
  bUsedWithNiagaraRibbons=true
  bUsedWithNiagaraSprites=true
  blendMode=BLEND_Translucent
  pixelDepthOffsetMode=PDOM_Legacy
  shadingModel=MSM_Unlit
  usedShadingModels=Unlit
 }
 nodes [
  n1 = MaterialFunctionCall!(Width=Constant() {r=0.5}, Sharpness=Constant(), Offset, Direction Switch=StaticBool() {value=true}, Compare, Input Coordinates) {materialFunction=/Engine/Functions/Engine_MaterialFunctions03/Procedurals/GeneratedBand}
  n2 = VectorParameter() {parameterName=Side Gradient Options, channelNames {a=Contrast, b=Offset, g=Multiply, r=Exponent}, defaultValue=<rgba 1 1 0 0>}
  n3 = Add(A=Multiply(A=Power(Base=@n1.Result, Exp=@n2.Exponent), B=@n2.Multiply), B=@n2.Offset)
  n4 = MaterialFunctionCall() {materialFunction=&LinearGradient}
  n5 = VectorParameter() {parameterName=Top Gradient Options, channelNames {a=Contrast, b=Offset, g=Multiply, r=Exponent}, defaultValue=<rgba 1 1 0 0>}
  n6 = Multiply(A=Power(Base=@n4.VGradient, Exp=@n5.Exponent), B=@n5.Multiply)
  n7 = VectorParameter() {parameterName=bottom Gradient Options, channelNames {a=Contrast, b=Offset, g=Multiply, r=Exponent}, defaultValue=<rgba 1 1 0 0>}
  n8 = Multiply(A=Power(Base=OneMinus(None=@n4.VGradient), Exp=@n7.Exponent), B=@n7.Multiply)
  n9 = Multiply(
   A=MaterialFunctionCall(In=Add(A=@n6, B=@n5.Offset), Contrast=@n5.Contrast) {materialFunction=&CheapContrast}.Result
   B=MaterialFunctionCall(In=Add(A=@n8, B=@n7.Offset), Contrast=@n7.Contrast) {materialFunction=&CheapContrast}.Result
  )
  Multiply(
   A=ParticleColor().A
   B=Multiply(
    A=MaterialFunctionCall(In=@n3, Contrast=@n2.Contrast) {materialFunction=&CheapContrast}.Result
    B=Multiply(B=@n9) {constA=4}
   )
  )
 ]
}
