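# Niagara effect for loot orb creation that throws out a warm golden energy wave with lingering smoke.
# Wave02 bursts a few randomly scaled, oriented meshes that expand while curves drive dynamic material parameters, a brightness flash and alpha fade.
# LingerSmoke emits a short-delayed burst of drifting sprites that slow, spin, grow and fade through SubUV frames.
# Water-related user variables are declared but not referenced by the shown modules.
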NiagaraSystem /Skyfire/Gameplay/Orbs/FX/Loot/NS_LootOrbCreation_EnergyWave
dependencies {
 usedDataInterfaces [&NiagaraDataInterfaceColorCurve, &NiagaraDataInterfaceCurve]
 usedDynamicInputs [&MakeVector4_FromLinearColor, &ColorFromCurve]
 usedModules [&SystemState]
}
emitters {
 LingerSmoke {
  emitterUpdateScript [
   &EmitterState(Life Cycle Mode=::NewEnumerator1, Loop Behavior=::NewEnumerator1, Loop Delay=0.1, Loop Duration=0.2, UseLoopDelay=true)
   &SpawnRate(SpawnRate=50, Use Spawn Probability=false) {enabled=false}
   &SpawnBurst_Instantaneous(Spawn Count=12, Spawn Time=0.1, Use Spawn Probability=false)
  ]
  particleSpawnScript [
   /Niagara/Modules/Spawn/Initialization/InitializeParticle(
    Color=&UniformRangedLinearColor(Minimum=<rgba 1 0.8181 0.4 0.1>, Maximum=<rgba 1 0.7143 0.4 0.15>, Color Channel Mode=::NewEnumerator1)
    Lifetime=&UniformRangedFloat(Minimum=2, Maximum=3)
    Module.Write Color=true
    Module.Write Lifetime=true
    Module.Write Mass=true
    Module.Write Material Random=true
    Module.Write Position=false
    Module.Write Scale=false
    Module.Write SpriteRotation=true
    Module.Write SpriteSize=true
    Position=&SimulationPosition()
    Sprite Rotation=&UniformRangedFloat(Minimum=0, Maximum=360)
    Sprite Size=&Vector2DFromFloat(Value=&UniformRangedFloat(Minimum=150, Maximum=300))
   )
   &CylinderLocation(
    Cylinder Height=0.5
    Cylinder Radius=40
    MaskSpawnGroup=false
    Module.Enable Lathe Profile=false
    Module.Override Randomness=false
    Module.Surface Only=true
    Offset=<0 0 25>
    Surface Only Band Thickness=1
   )
   &AddVelocityFromPoint(
    Module.Use Falloff Distance=false
    Origin Offset=<50 -50 15>
    Velocity Origin:Vector3f=&SimulationPosition()
    Velocity Strength=&FloatFromCurve(
     FloatCurve=DataInterface() {propertyValues {Curve {keys [<at 0.1 1 RCIM_Cubic>,<at 0.3 0>]}}}
     CurveIndex=$System.Age
     Scale Curve=&UniformRangedFloat(Minimum=1500, Maximum=2250)
    )
   ) {enabled=false}
   &CameraOffset(Camera Offset Amount=50) {module=CameraOffset001}
   &AddVelocity(Velocity=&RandomRangeVector(Minimum=<-250 -250 150>, Maximum=<250 250 -500>))
  ]
  particleUpdateScript [
   &ParticleState()
   &Drag(Drag=2, Use Linear Drag=true, Use Rotational Drag=false)
   &SpriteRotationRate(Rotation Rate=&Multiply_Float(
     A=&Lerp_Float(A=-200, B=200, Alpha=$Particles.MaterialRandom)
     B=&FloatFromCurve(FloatCurve=DataInterface() {propertyValues {Curve {keys [
          {arriveTangent=-4.4572, interpMode=RCIM_Cubic, leaveTangent=-4.4572, tangentMode=RCTM_User, time=0, value=1}
          {arriveTangent=-1.0628, interpMode=RCIM_Cubic, leaveTangent=-1.0628, tangentMode=RCTM_User, time=0.2, value=0.4}
          <at 1.0011 0.01>
         ]}}})
    ))
   &SolveForcesAndVelocity(Clamp Velocity=false, Limit Acceleration=false)
   &ScaleSpriteSize(Scale Factor=&Vector2DFromFloat(Value=&FloatFromCurve(FloatCurve=DataInterface() {propertyValues {Curve {keys [
          {arriveTangent=7.4819, interpMode=RCIM_Cubic, leaveTangent=7.4819, tangentMode=RCTM_User, time=0, value=0.4215}
          {arriveTangent=0.5944, interpMode=RCIM_Cubic, leaveTangent=0.5944, tangentMode=RCTM_User, time=0.1164, value=0.8874}
          {interpMode=RCIM_Cubic, tangentMode=RCTM_User, time=1, value=1}
         ]}}})))
   &ScaleColor(
    Scale Alpha=&FloatFromCurve(FloatCurve=DataInterface() {propertyValues {Curve {keys [
         {arriveTangent=21.9814, interpMode=RCIM_Cubic, leaveTangent=21.9814, tangentMode=RCTM_User, time=0, value=0}
         {arriveTangent=0.1895, interpMode=RCIM_Cubic, leaveTangent=0.1895, tangentMode=RCTM_User, time=0.1, value=0.75}
         {arriveTangent=-2.5555, interpMode=RCIM_Cubic, leaveTangent=-2.5555, tangentMode=RCTM_User, time=1, value=0}
        ]}}})
    Scale RGB=&VectorFromFloat(Value=&FloatFromCurve(FloatCurve=DataInterface() {propertyValues {Curve {keys [{arriveTangent=-0.1872, interpMode=RCIM_Cubic, leaveTangent=-0.1872, tangentMode=RCTM_User, time=0, value=2}, <at 1 1>]}}}))
    ScaleA=true
    ScaleRGB=true
   )
   /SetVariables(Particles.SubImageIndex=&Lerp_Float(A=25, B=63, Alpha=$Particles.NormalizedAge)) {isSetParameters=true}
  ]
  rendererClasses [&NiagaraSpriteRendererProperties]
  renderers [{rendererClass=&NiagaraSpriteRendererProperties}]
 }
 Wave02 {
  emitterUpdateScript [
   &EmitterState()
   &SpawnBurst_Instantaneous(Spawn Count=&RandomRangeInt(Minimum=2, Maximum=3), Use Spawn Probability=false)
  ]
  particleSpawnScript [
   /Niagara/Modules/Spawn/Initialization/V2/InitializeParticle(
    Color Maximum=<rgba 1 0.8 0.2 0.75>
    Color Minimum=<rgba 1 0.8181 0.4 0.5>
    Color Mode=::NewEnumerator2
    Lifetime=2.5
    Mesh Scale=&Multiply_VectorByFloat(Vector=&RandomRangeVector(Minimum=<0.4 0.4 0.4>, Maximum=<1.5 1.5 1.5>), Float=4)
    Mesh Scale Mode=::NewEnumerator4
    UsePositionOffset=false
    Write Material Random=true
   )
   &InitialMeshOrientation(
    Orientation Axis=<0 0 1>
    Rotation=&RandomRangeVector(Minimum=<0 0 0>, Maximum=<1 1 1>)
    Use Orientation Vector=true
    Use Rotation Vector=true
   )
   &CameraOffset(Camera Offset Amount=100)
   &SystemLocation(MaskSpawnGroup=false, Offset=<0 0 50>)
  ]
  particleUpdateScript [
   &ParticleState()
   &ScaleColor(
    Linear Color Curve=DataInterface() {propertyValues {
      AlphaCurve {keys [<at 0 1>,<at 0.4343 1>,<at 1 0>]}
      BlueCurve {keys [<at 0.0133 1>,<at 0.1611 1>,<at 0.6544 0.394>]}
      GreenCurve {keys [<at 0.0133 1>,<at 0.1611 1>,<at 0.6544 0.394>]}
      RedCurve {keys [<at 0.0133 1>,<at 0.1611 1>,<at 0.6544 0.394>]}
     }}
    Scale Mode=::NewEnumerator2
   )
   &DynamicMaterialParameters(
    Index 0 Param 1=&FloatFromCurve(FloatCurve=DataInterface() {propertyValues {Curve {keys [{arriveTangent=-4.0141, interpMode=RCIM_Cubic, leaveTangent=-4.0141, tangentMode=RCTM_User, time=0, value=0.6}, <at 1 -0.7>]}}})
    Index 0 Param 2=&FloatFromCurve(FloatCurve=DataInterface() {propertyValues {Curve {keys [{arriveTangent=-0.5818, interpMode=RCIM_Cubic, leaveTangent=-0.5818, tangentMode=RCTM_User, time=0, value=1}, <at 1 0.9>]}}})
    Index 0 Param 3=&FloatFromCurve(FloatCurve=DataInterface() {propertyValues {Curve {keys [<at 0 0>,<at 0.15 1>,<at 1 1.2>]}}})
    Module.Param0WriteEnabled=true
    Module.Param1WriteEnabled=true
    Module.Param2WriteEnabled=true
    Module.Param3WriteEnabled=false
   )
   &ScaleColor(
    Scale RGB=&VectorFromFloat(Value=&FloatFromCurve(FloatCurve=DataInterface() {propertyValues {Curve {keys [<at 0 0>, {arriveTangent=-15.019, interpMode=RCIM_Cubic, leaveTangent=-15.019, tangentMode=RCTM_User, time=0.2, value=6}, <at 1 1>]}}}))
    ScaleA=true
    ScaleRGB=true
   ) {module=ScaleColor001}
   &ScaleMeshSize(Scale Factor=&VectorFromFloat(Value=&FloatFromCurve(FloatCurve=DataInterface() {propertyValues {Curve {keys [<at 0 0>,<at 0.0694 1.0545>,<at 1 3>]}}})))
  ]
  rendererClasses [&NiagaraMeshRendererProperties]
  renderers [{rendererClass=&NiagaraMeshRendererProperties, rendererIndex=0}]
 }
}
summary {
 systemName=NS_LootOrbCreation_EnergyWave
 emitters [
  {emitterName=Wave02, rendererClasses [&NiagaraMeshRendererProperties]}
  {emitterName=LingerSmoke, rendererClasses [&NiagaraSpriteRendererProperties]}
 ]
 userVariables [
  User.LoopFoamSpawn {type=NiagaraFloat, defaultValue=0, description=""}
  User.WaterBodyIndex {type=NiagaraFloat, defaultValue=0, description=""}
  User.WaterDepth {type=NiagaraFloat, defaultValue=0, description=""}
  User.WaterPlaneLocation {type=Vector3f, description="", defaultValue {struct=Vector3f, value=<0 0 0>}}
  User.WaterPlaneNormal {type=Vector3f, description="", defaultValue {struct=Vector3f, value=<0 0 0>}}
  User.WaveAttenuationFactor {type=NiagaraFloat, defaultValue=0, description=""}
  User.WaveTime {type=NiagaraFloat, defaultValue=0, description=""}
 ]
}
