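# Niagara system for the idle look of a golden loot orb, with an interacted-with state.
# Layers a mesh sphere with pulsing light, icon sprite, fake glow, rising diamonds and squares; the Energy emitter is disabled.
# System timers, alongside a bIsBeingInteractedWith user flag, swap rising particles for interaction rings, jitter sphere and icon, and shrink sphere and glow.
# Distance culling trims several emitters, while a corona sprite fades in and grows with camera distance.
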NiagaraSystem /Skyfire/Gameplay/Orbs/FX/Loot/NS_LootOrb_Idle
dependencies {
 usedDataInterfaces [&NiagaraDataInterfaceCurve, &NiagaraDataInterfaceCamera, &NiagaraDataInterfacePlatformSet]
 usedDynamicInputs [&Boolean_NotOperation]
 usedModules [&SystemState, &FN_Timer]
}
emitters {
 Corona_Ring {
  emitterUpdateScript [
   &EmitterState()
   &SpawnBurst_Instantaneous(Age:NiagaraFloat=$Emitter.Age, Spawn Count=1, Use Spawn Probability=false)
   /SetVariables(Emitter.CamPos:Vector3f=&ReturnCameraVector(Camera Property=::NewEnumerator0, Coordinate Space=::Simulation)) {isSetParameters=true, module=SetVariables_#1}
  ]
  particleSpawnScript [&InitializeParticle(
    Color=<rgba 1 0.7259 0 0.018>
    Lifetime=10
    Sprite Size Mode=::NewEnumerator3
    Uniform Sprite Size=2000
    UsePositionOffset=false
    Write Material Random=true
   )]
  particleUpdateScript [
   &ParticleState(
    Kill Particles When Lifetime Has Elapsed=false
    Let Infinitely Lived Particles Die When Emitter Deactivates=true
    Loop Particles Lifetime=true
   )
   &ScaleColor(
    Scale Alpha=&Lerp_Float(
     A=0
     B=2
     Alpha=&FloatFromCurve(
      FloatCurve=DataInterface() {propertyValues {Curve {keys [<at 0 0 RCIM_Constant>, <at 0.325 0>, {tangentMode=RCTM_User, time=1, value=1}]}}}
      CurveIndex=&NormalizeDistanceRange(
       Start Position:NiagaraPosition=Linked() {linkedVariable {name=Emitter.CamPos, type=Vector3f}}
       End Position:NiagaraPosition=/Niagara/DynamicInputs/Helpers/SimulationPosition()
       Distance:NiagaraFloat=10000
       Invert Normalized Range:NiagaraBool=false
      )
     )
    )
    ScaleA=true
    ScaleRGB=true
   )
   &ScaleSpriteSize(Scale Factor=&Vector2DFromFloat(Value=&Lerp_Float(
      A=0
      B=1
      Alpha=&FloatFromCurve(
       FloatCurve=DataInterface() {propertyValues {Curve {keys [<at 0 0 RCIM_Constant>, {interpMode=RCIM_Cubic, tangentMode=RCTM_Break, time=0.32, value=0.75}, {tangentMode=RCTM_User, time=1, value=1}]}}}
       CurveIndex=&NormalizeDistanceRange(
        Start Position:NiagaraPosition=Linked() {linkedVariable {name=Emitter.CamPos, type=Vector3f}}
        End Position:NiagaraPosition=&SimulationPosition()
        Distance:NiagaraFloat=10000
        Invert Normalized Range:NiagaraBool=false
       )
      )
     )))
  ]
  rendererClasses [&NiagaraSpriteRendererProperties]
  renderers [{rendererClass=&NiagaraSpriteRendererProperties}]
 }
 Daimonds {
  emitterUpdateScript [
   &EmitterState(
    Enable Distance Culling=true
    Life Cycle Mode=::NewEnumerator1
    Loop Duration=2
    MaxDistance=3200
    Module.bApplySpawnCountScaleByDistance=false
    Module.bUseMaxDistance=true
    Module.bUseMinDistance=false
    Scalability Mode=::NewEnumerator1
    Scale Spawn Count=true
    UseLoopDelay=false
   )
   &SpawnRate(SpawnRate=&Lerp_Float(A=10, B=0, Alpha=$System.FN_Timer.Value), Use Spawn Probability=false)
  ]
  particleSpawnScript [
   &InitializeParticle(
    Color=<rgba 0.8148 0.6666 0.3259 0.25>
    Lifetime Min=2.5
    Lifetime Mode=::NewEnumerator1
    Mass Mode=::NewEnumerator3
    Sprite Rotation Mode=::NewEnumerator2
    Sprite Size Max=<10 100>
    Sprite Size Min=<2 50>
    Sprite Size Mode=::NewEnumerator2
    UsePositionOffset=false
    Write Material Random=true
   )
   &SphereLocation(
    Hemisphere Z=true
    MaskSpawnGroup=false
    Override Seed=false
    Sphere Radius=60
    Surface Only=true
    UseNonUniformScale=false
    UseOffset=false
    UseSphereOrientation=false
   )
   &AddVelocity(Velocity=&MakeVector(X=0, Y=0, Z=&RandomRangeFloat(Minimum=50, Maximum=100)))
  ]
  particleUpdateScript [
   &ParticleState()
   &Drag(Drag=0.1, Use Linear Drag=true, Use Rotational Drag=false)
   &ScaleColor(
    Scale Alpha=&FloatFromCurve(FloatCurve=DataInterface() {propertyValues {Curve {keys [
         {interpMode=RCIM_Cubic, tangentMode=RCTM_User, time=0, value=0}
         {interpMode=RCIM_Cubic, tangentMode=RCTM_User, time=0.5, value=1}
         {tangentMode=RCTM_User, time=1, value=0}
        ]}}})
    ScaleA=true
    ScaleRGB=true
   )
   &SolveForcesAndVelocity(Clamp Velocity=false, Limit Acceleration=false)
   &ScaleSpriteSize(Scale Factor=&Vector2DFromFloat(Value=&FloatFromCurve(FloatCurve=DataInterface() {propertyValues {Curve {keys [
          {interpMode=RCIM_Cubic, tangentMode=RCTM_User, time=0, value=0}
          {interpMode=RCIM_Cubic, tangentMode=RCTM_User, time=0.5, value=1}
          {tangentMode=RCTM_User, time=1, value=0.75}
         ]}}})))
  ]
  rendererClasses [&NiagaraSpriteRendererProperties]
  renderers [{rendererClass=&NiagaraSpriteRendererProperties}]
 }
 Energy {
  bEnabled=false
  emitterUpdateScript [
   &EmitterState(Life Cycle Mode=::NewEnumerator1, Loop Duration=2, UseLoopDelay=false)
   &SpawnRate(SpawnRate=2, Use Spawn Probability=false)
   &SpawnBurst_Instantaneous(Spawn Count=1, Use Spawn Probability=false)
  ]
  particleSpawnScript [
   &InitializeParticle(
    Color Maximum=<rgba 1 0.5239 0 1.5>
    Color Minimum=<rgba 1 0.8181 0.4 1>
    Color Mode=::NewEnumerator2
    Lifetime Max=1.75
    Lifetime Min=1.4
    Lifetime Mode=::NewEnumerator1
    Mass Mode=::NewEnumerator3
    Mesh Scale Max=<10 16 14>
    Mesh Scale Min=<10 16 14>
    Mesh Scale Mode=::NewEnumerator2
    Sprite Rotation Angle=90
    Sprite Rotation Mode=::NewEnumerator0
    Sprite Size Mode=::NewEnumerator1
    Sprite UV Mode=::NewEnumerator1
    Uniform Sprite Size Max=100
    Uniform Sprite Size Min=100
    UsePositionOffset=true
    Write Material Random=true
   )
   /SetVariables(Particles.Velocity=<0 0 1>) {isSetParameters=true}
   &InitialMeshOrientation(Orientation Axis=<0 0 -1>, Use Orientation Vector=true, Use Rotation Vector=false)
  ]
  particleUpdateScript [
   &ParticleState()
   &ScaleColor(
    Scale Alpha=&FloatFromCurve(FloatCurve=DataInterface() {propertyValues {Curve {keys [
         {interpMode=RCIM_Cubic, tangentMode=RCTM_User, time=0, value=0}
         {interpMode=RCIM_Cubic, tangentMode=RCTM_User, time=0.5, value=1}
         {tangentMode=RCTM_User, time=1, value=0}
        ]}}})
    ScaleA=true
    ScaleRGB=true
   )
   &SolveForcesAndVelocity(Clamp Velocity=false, Limit Acceleration=false)
   &ScaleMeshSize(Scale Factor=&Lerp_Vector(A=<1 1 1>, B=<0.55 0.8 1>, Alpha=$System.FN_Timer001.Value))
  ]
  rendererClasses [&NiagaraMeshRendererProperties]
  renderers [{rendererClass=&NiagaraMeshRendererProperties, rendererIndex=0}]
 }
 Fake_Glow {
  emitterUpdateScript [
   &EmitterState(
    Enable Distance Culling=true
    Life Cycle Mode=::NewEnumerator1
    Max Distance Response=::KillImmediately
    MaxDistance=2500
    Module.bApplySpawnCountScaleByDistance=true
    Module.bUseMaxDistance=true
    Module.bUseMinDistance=false
    Reset Age On Awaken=true
    Scalability Mode=::NewEnumerator1
    Scale Spawn Count=true
    Spawn Count Scale By Distance=DataInterface() {propertyValues {Curve {keys [<at 0 1>,<at 0.75 1>,<at 1 0.25>]}}}
    UseLoopDelay=false
   )
   &SpawnRate(SpawnRate=4, Use Spawn Probability=false)
   /SetVariables(Emitter.CamPos:Vector3f=&ReturnCameraVector(Camera Property=::NewEnumerator0, Coordinate Space=::Simulation)) {isSetParameters=true, module=SetVariables_#1}
  ]
  particleSpawnScript [
   &InitializeParticle(
    Color Maximum=<rgba 15 7.8578 0 1>
    Color Minimum=<rgba 15 12.2715 6 1>
    Color Mode=::NewEnumerator2
    Lifetime Max=4
    Lifetime Min=5
    Lifetime Mode=::NewEnumerator1
    Mass Mode=::NewEnumerator3
    Sprite Rotation Mode=::NewEnumerator2
    Sprite Size Mode=::NewEnumerator3
    Uniform Sprite Size=172.5
    UsePositionOffset=false
    Write Material Random=true
   )
   &SphereLocation(
    MaskSpawnGroup=false
    Override Seed=false
    Sphere Radius=4
    Surface Only=false
    UseNonUniformScale=false
    UseOffset=false
    UseSphereOrientation=false
   )
  ]
  particleUpdateScript [
   &ParticleState()
   &Drag(Drag=0.25, Use Linear Drag=true, Use Rotational Drag=false)
   &ScaleColor(
    Scale Alpha=&FloatFromCurve(
     FloatCurve=DataInterface() {propertyValues {Curve {keys [
         {interpMode=RCIM_Cubic, tangentMode=RCTM_User, time=0, value=0}
         {interpMode=RCIM_Cubic, tangentMode=RCTM_User, time=0.5, value=1}
         {tangentMode=RCTM_User, time=1, value=0}
        ]}}}
     Scale Curve=&MakeCustomFloatFromBool(
      Bool=&IsPlatformSetActive(New Platform Set=DataInterface() {propertyValues {Platforms {qualityLevelMask=-29}}})
      True Float=5
      False Float=1
     )
    )
    ScaleA=true
    ScaleRGB=true
   )
   &SolveForcesAndVelocity(Clamp Velocity=false, Limit Acceleration=false)
   &ScaleSpriteSize(Scale Factor=&Vector2DFromFloat(Value=&Multiply_Float(
      A=&FloatFromCurve(FloatCurve=DataInterface() {propertyValues {Curve {keys [
           {interpMode=RCIM_Cubic, leaveTangentWeight=0.4835, time=0, value=0}
           {arriveTangentWeight=0.5027, interpMode=RCIM_Cubic, leaveTangentWeight=0.4835, time=0.5, value=1}
           {arriveTangentWeight=0.5027, interpMode=RCIM_Cubic, time=1, value=0}
          ]}}})
      B=&Lerp_Float(A=1, B=0.85, Alpha=$System.FN_Timer001.Value)
     )))
  ]
  rendererClasses [&NiagaraSpriteRendererProperties]
  renderers [{rendererClass=&NiagaraSpriteRendererProperties}]
 }
 Icon {
  emitterUpdateScript [
   &EmitterState(
    Enable Distance Culling=true
    Life Cycle Mode=::NewEnumerator1
    Max Distance Response=::SleepAndClearParticles
    MaxDistance=15000
    Module.bUseMaxDistance=true
    Module.bUseMinDistance=false
    Reset Age On Awaken=true
    Scalability Mode=::NewEnumerator1
    UseLoopDelay=false
   )
   &SpawnBurst_Instantaneous(Spawn Count=1, Use Spawn Probability=false)
  ]
  particleSpawnScript [
   &InitializeParticle(
    Color=<rgba 3.6 3.015 1.26 0.75>
    Lifetime=5
    Sprite Size Mode=::NewEnumerator3
    Uniform Sprite Size=65
    UsePositionOffset=false
    Write Material Random=true
   )
   &CameraOffset(Camera Offset Amount=85)
   /SetVariables(Particles.BG Color:LinearColor=<rgba 0 0 0 0.75>, Particles.BG Sprite Size:Vector2f=<150 150>) {isSetParameters=true}
   &DynamicMaterialParameters(
    Index 0 Param 1=1
    Module.Param0WriteEnabled=true
    Module.Param1WriteEnabled=true
    Module.Param2WriteEnabled=true
    Module.Param3WriteEnabled=true
   )
  ]
  particleUpdateScript [
   &ParticleState()
   &SphereLocation(
    Evaluation Type:ENiagaraRandomnessEvaluation=::NewEnumerator1
    MaskSpawnGroup=false
    Override Seed=false
    Sphere Radius=&Multiply_Float(A=$System.FN_Timer001.Value, B=1.5)
    Surface Only=false
    UseNonUniformScale=false
    UseOffset=false
    UseSphereOrientation=false
   )
  ]
  rendererClasses [&NiagaraSpriteRendererProperties]
  renderers [
   {rendererClass=&NiagaraSpriteRendererProperties}
   {rendererClass=&NiagaraSpriteRendererProperties, rendererIndex=1}
  ]
 }
 Interaction_Rings {
  emitterUpdateScript [
   &EmitterState(Life Cycle Mode=::NewEnumerator1, Loop Duration=1, UseLoopDelay=false)
   &SpawnRate(SpawnRate=&Lerp_Float(A=0, B=8, Alpha=$System.FN_Timer.Value), Use Spawn Probability=false)
  ]
  particleSpawnScript [
   &InitializeParticle(
    AdjustAlpha=false
    AdjustHue=true
    AdjustSaturation=false
    AdjustValue=false
    Color=<rgba 1 1 0.594 0.372>
    Color Mode=::NewEnumerator3
    Hue Shift Range:Vector2f=<-0.05 0.1>
    Lifetime=0.75
    Mesh Scale=<3 3 3>
    Mesh Scale Mode=::NewEnumerator4
    UsePositionOffset=false
    Write Material Random=true
   )
   &InitialMeshOrientation(
    Orientation Axis=<0 0 1>
    Rotation=&RandomRangeVector(Minimum=<0.025 0.025 0>, Maximum=<-0.025 -0.025 1>)
    Use Orientation Vector=true
    Use Rotation Vector=true
   )
  ]
  particleUpdateScript [
   &ParticleState()
   &self:Line_Location(End Pos:Vector3f=<0 0 80>, Pos On Line:NiagaraFloat=$Particles.NormalizedAge, Start Pos:Vector3f=<0 0 -80>) {module=Line_Location}
   &ScaleColor(
    Scale Alpha=&FloatFromCurve(FloatCurve=DataInterface() {propertyValues {Curve {keys [
         {interpMode=RCIM_Cubic, leaveTangent=5.0189, tangentMode=RCTM_Break, time=0, value=0}
         {
          arriveTangentWeight=0.457
          interpMode=RCIM_Cubic
          leaveTangentWeight=0.5403
          tangentMode=RCTM_Break
          tangentWeightMode=RCTWM_WeightedBoth
          time=0.5
          value=1
         }
         {arriveTangent=-3.838, interpMode=RCIM_Cubic, tangentMode=RCTM_Break, time=1, value=0}
        ]}}})
    ScaleA=true
    ScaleRGB=false
   ) {enabled=false}
   &ScaleMeshSize(Scale Factor=&VectorFromFloat(Value=&FloatFromCurve(FloatCurve=DataInterface() {propertyValues {Curve {keys [
          {
           arriveTangent=11.3858
           arriveTangentWeight=0.9989
           interpMode=RCIM_Cubic
           leaveTangent=11.3858
           leaveTangentWeight=0.9989
           tangentMode=RCTM_User
           tangentWeightMode=RCTWM_WeightedBoth
           time=0
           value=0
          }
          {
           arriveTangentWeight=0.1239
           interpMode=RCIM_Cubic
           leaveTangentWeight=0.1239
           tangentMode=RCTM_User
           tangentWeightMode=RCTWM_WeightedBoth
           time=0.5
           value=1
          }
          {
           arriveTangent=-5.7311
           arriveTangentWeight=0.9696
           interpMode=RCIM_Cubic
           leaveTangent=-5.7311
           tangentMode=RCTM_User
           tangentWeightMode=RCTWM_WeightedBoth
           time=1
           value=0
          }
         ]}}})))
   &SolveForcesAndVelocity(Clamp Velocity=false, Limit Acceleration=false)
  ]
  rendererClasses [&NiagaraMeshRendererProperties]
  renderers [{rendererClass=&NiagaraMeshRendererProperties, rendererIndex=0}]
 }
 Sphere {
  emitterUpdateScript [
   &EmitterState()
   &SpawnRate(Use Spawn Probability=false) {enabled=false}
   &SpawnBurst_Instantaneous(Age:NiagaraFloat=$Emitter.Age, Spawn Count=1, Use Spawn Probability=false)
  ]
  particleSpawnScript [&InitializeParticle(
    Color=<rgba 0.15 0.115 0 1>
    Lifetime=5
    Mass Mode=::NewEnumerator3
    Mesh Scale Mode=::NewEnumerator3
    Mesh Uniform Scale=3.25
    Sprite Rotation Mode=::NewEnumerator2
    Sprite Size Mode=::NewEnumerator3
    Uniform Sprite Size=125
    UsePositionOffset=false
    Write Material Random=true
   )]
  particleUpdateScript [
   &ParticleState(Kill Particles When Lifetime Has Elapsed=false)
   &FN_Light_Attributes(
    Color=<rgba 0.8148 0.6204 0.1736 1>
    Create Color Attribute=true
    Create Exponent Attribute=true
    Create Radius Attribute=true
    Exponent=3
    Radius=175
    Scale Color=&Lerp_Float(A=5, B=10, Alpha=&Sine(Normalized Angle=$Emitter.Age, Period=3, Scale=0.5, Bias=0.5))
   )
   &SphereLocation(
    Evaluation Type:ENiagaraRandomnessEvaluation=::NewEnumerator1
    MaskSpawnGroup=false
    Override Seed=false
    Sphere Radius=&Multiply_Float(A=$System.FN_Timer001.Value, B=1.5)
    Surface Only=false
    UseNonUniformScale=false
    UseOffset=false
    UseSphereOrientation=false
   )
   &ScaleMeshSize(Scale Factor=&Multiply_Vector(
     A=&Lerp_Vector(
      A=<1 1 1>
      B=&RandomRangeVector(Minimum=<0.95 0.95 0.95>, Maximum=<1 1 1>, Evaluation Type=::NewEnumerator1)
      Alpha=$System.FN_Timer001.Value
     )
     B=&VectorFromFloat(Value=&FloatFromCurve(
       FloatCurve=DataInterface() {propertyValues {Curve {keys [<at 0 1 RCIM_Cubic>, {interpMode=RCIM_Cubic, leaveTangent=-0.9753, time=1, value=0.9}]}}}
       CurveIndex=$System.FN_Timer001.Value
      ))
    ))
  ]
  rendererClasses [&NiagaraMeshRendererProperties, &NiagaraLightRendererProperties]
  renderers [
   {rendererClass=&NiagaraMeshRendererProperties, rendererIndex=0}
   {rendererClass=&NiagaraLightRendererProperties, rendererIndex=1}
   {rendererClass=&NiagaraMeshRendererProperties, rendererIndex=2}
  ]
 }
 Squares {
  emitterUpdateScript [
   &EmitterState(
    Enable Distance Culling=true
    Life Cycle Mode=::NewEnumerator1
    Loop Duration=2
    MaxDistance=3200
    Module.bApplySpawnCountScaleByDistance=false
    Module.bUseMaxDistance=true
    Module.bUseMinDistance=false
    Scalability Mode=::NewEnumerator1
    Scale Spawn Count=true
    UseLoopDelay=false
   )
   &SpawnRate(SpawnRate=&Lerp_Float(A=22.5, B=0, Alpha=$System.FN_Timer.Value), Use Spawn Probability=false)
  ]
  particleSpawnScript [
   &InitializeParticle(
    Color Minimum=<rgba 0.8148 0.6666 0.3259 1>
    Color Mode=::NewEnumerator2
    Lifetime Max=1.5
    Lifetime Mode=::NewEnumerator1
    Mass Mode=::NewEnumerator3
    Mesh Scale Max=<12 12 2>
    Mesh Scale Min=<7 7 1>
    Mesh Scale Mode=::NewEnumerator2
    Sprite Rotation Mode=::NewEnumerator2
    Sprite Size Mode=::NewEnumerator1
    Uniform Sprite Size Max=12
    Uniform Sprite Size Min=6
    UsePositionOffset=false
    Write Material Random=true
   )
   &SphereLocation(
    MaskSpawnGroup=false
    Sphere Distribution:ENiagaraSphereDistributionMode=::NewEnumerator1
    Sphere Radius=65
    U Position:NiagaraFloat=&RandomRangeFloat(Minimum=0, Maximum=1)
    UseNonUniformScale=false
    UseOffset=false
    UseSphereOrientation=false
    V Position:NiagaraFloat=&RandomRangeFloat(Minimum=0.859, Maximum=1)
   )
   &AddVelocityInCone(
    Cone Angle=20
    Cone Axis=<0 0 1>
    Module.Use Velocity Falloff On Cone Axis=false
    Velocity Distribution Along Cone Axis=0.75
    Velocity Strength=&RandomRangeFloat(Minimum=500, Maximum=850)
   ) {enabled=false}
   &AddVelocity(Velocity=<0 0 100>)
   /NiagaraFN/Sprite/FN_SpriteAlignmentAndFacing(SpriteAlignment:Vector3f=<0 0 1>, SpriteFacing:Vector3f=$Output.SphereLocation.SphereVector)
   &InitialMeshOrientation(Orientation Axis=<0 0 1>, Orientation Vector=$Output.SphereLocation.SphereNormal, Use Orientation Vector=true, Use Rotation Vector=false)
  ]
  particleUpdateScript [
   &ParticleState()
   &GravityForce() {enabled=false}
   &Drag(Drag=0.25, Use Linear Drag=true, Use Rotational Drag=false) {enabled=false}
   &MeshRotationForce(Rotation=&RandomRangeVector(Minimum=<-2 -2 -2>, Maximum=<2 2 2>))
   &ScaleColor(
    Scale Alpha=&FloatFromCurve(FloatCurve=DataInterface() {propertyValues {Curve {keys [
         {interpMode=RCIM_Cubic, tangentMode=RCTM_User, time=0, value=0}
         {interpMode=RCIM_Cubic, tangentMode=RCTM_User, time=0.1, value=1}
         {tangentMode=RCTM_User, time=1, value=0}
        ]}}})
    ScaleA=true
    ScaleRGB=true
   )
   &CurlNoiseForce(Noise Frequency=5, Noise Quality / Cost=::NewEnumerator0, Noise Strength=50, Pan Noise=true, Pan Noise Field=<1 1 1>)
   &SolveForcesAndVelocity(Clamp Velocity=false, Limit Acceleration=false)
   &SolveRotationalForcesAndVelocity()
   /SetVariables(Particles.SpriteFacing=/Niagara/DynamicInputs/Vectors/RotateVector(
     Vector To Rotate:Vector3f=$Particles.Initial.SpriteFacing
     Yaw:NiagaraFloat=0
     Pitch:NiagaraFloat=0
     Roll:NiagaraFloat=90
     Rotation Rate:NiagaraFloat=$Engine.DeltaTime
    )) {isSetParameters=true}
  ]
  rendererClasses [&NiagaraSpriteRendererProperties, &NiagaraMeshRendererProperties]
  renderers [
   {rendererClass=&NiagaraSpriteRendererProperties}
   {rendererClass=&NiagaraMeshRendererProperties, rendererIndex=1}
  ]
 }
}
summary {
 systemName=NS_LootOrb_Idle
 emitters [
  {emitterName=Fake_Glow, rendererClasses [&NiagaraSpriteRendererProperties]}
  {emitterName=Daimonds, rendererClasses [&NiagaraSpriteRendererProperties]}
  {emitterName=Squares, rendererClasses [&NiagaraSpriteRendererProperties, &NiagaraMeshRendererProperties]}
  {bEnabled=false, emitterName=Energy, rendererClasses [&NiagaraMeshRendererProperties]}
  {emitterName=Sphere, rendererClasses [&NiagaraMeshRendererProperties, &NiagaraLightRendererProperties]}
  {emitterName=Icon, rendererClasses [&NiagaraSpriteRendererProperties]}
  {emitterName=Interaction_Rings, rendererClasses [&NiagaraMeshRendererProperties]}
  {emitterName=Corona_Ring, rendererClasses [&NiagaraSpriteRendererProperties]}
 ]
 userVariables [User.bIsBeingInteractedWith {type=NiagaraBool}]
}
