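# Niagara system for a loot orb position indicator: a single golden, translucent mesh beam named PickupBeam, rotated 90 degrees.
# One looping particle fades in with emitter age, while a system FN_Timer value blends its color and boosts its alpha.
# Curves over a normalized range built from simulation position and scalability distance fraction drive alpha and a dynamic material parameter.
# Distance culling clears it at long range; a user bool exposes interaction state.
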NiagaraSystem /Skyfire/Gameplay/Orbs/FX/Loot/NS_Skyfire_LootOrb_Pos_Indicator
dependencies {
 usedDataInterfaces [&NiagaraDataInterfaceCurve]
 usedDynamicInputs [&Boolean_NotOperation]
 usedModules [&FN_Timer, &SystemState]
}
emitters {PickupBeam {
  emitterUpdateScript [
   &EmitterState(
    Enable Distance Culling=true
    Inactive Response=::NewEnumerator1
    Life Cycle Mode=::NewEnumerator1
    Loop Duration=3.5
    Max Distance Response=::SleepAndClearParticles
    MaxDistance=35000
    Module.bUseMaxDistance=true
    Module.bUseMinDistance=false
    Reset Age On Awaken=true
    Scalability Mode=::NewEnumerator1
    UseLoopDelay=true
   )
   &SpawnBurst_Instantaneous(Age:NiagaraFloat=$Emitter.Age, Spawn Count=1, Use Spawn Probability=false)
  ]
  particleSpawnScript [
   &InitializeParticle(
    Color Mode=::NewEnumerator0
    Lifetime=4
    Mesh Scale=<7 1 1>
    Mesh Scale Mode=::NewEnumerator4
    UsePositionOffset=false
    Write Material Random=true
   ) {module=InitializeParticle001}
   /SetVariables(Particles.MeshOrientation=&MakeQuatFromAxisAngle(Axis=<0 -1 0>, AngleInDegrees=90)) {isSetParameters=true}
   &CameraOffset(Camera Offset Amount=-20)
  ]
  particleUpdateScript [
   &Color(
    Color=&LerpLinearColors(StartColor=<rgba 1.1 0.8209 0.3025 0.25>, EndColor=<rgba 0.9823 0.7612 0.2456 0.25>, LerpFactor=$System.FN_Timer.Value)
    Scale Alpha=&Multiply_Float(
     A=&FloatFromCurve(
      FloatCurve=DataInterface() {propertyValues {Curve {keys [
          {interpMode=RCIM_Cubic, leaveTangentWeight=0.0833, tangentMode=RCTM_User, time=0, value=0}
          {arriveTangentWeight=0.1389, interpMode=RCIM_Cubic, leaveTangentWeight=0.2778, tangentMode=RCTM_User, time=0.1, value=1}
         ]}}}
      CurveIndex=$Emitter.Age
      Scale Curve=&Lerp_Float(A=1, B=10, Alpha=$System.FN_Timer.Value)
     )
     B=&FloatFromCurve(
      FloatCurve=DataInterface() {propertyValues {Curve {keys [
          <at 0.22 0.5 RCIM_Cubic>
          {arriveTangentWeight=0.3012, interpMode=RCIM_Cubic, leaveTangentWeight=0.1309, tangentMode=RCTM_User, time=0.75, value=1}
          {interpMode=RCIM_Cubic, leaveTangent=3.4262, time=1, value=1}
         ]}}}
      CurveIndex=&NormalizeDistanceRange(
       Start Position:NiagaraPosition=/Niagara/DynamicInputs/Helpers/SimulationPosition()
       End Position:NiagaraPosition=&VectorFromFloat(Value=&Multiply_Float(A=$Emitter.Scalability.DistanceFraction, B=15000))
       Distance:NiagaraFloat=15000
       Invert Normalized Range:NiagaraBool=false
      )
     )
    )
   )
   &ParticleState(
    Kill Particles When Lifetime Has Elapsed=false
    Let Infinitely Lived Particles Die When Emitter Deactivates=true
    Loop Particles Lifetime=true
   )
   &DynamicMaterialParameters(
    Index 0 Param 1=&Lerp_Float(
     A=150
     B=500
     Alpha=&FloatFromCurve(
      FloatCurve=DataInterface() {propertyValues {Curve {keys [
          <at 0.22 0 RCIM_Cubic>
          {arriveTangentWeight=0.3012, interpMode=RCIM_Cubic, leaveTangentWeight=0.1309, tangentMode=RCTM_User, time=0.75, value=1}
          {interpMode=RCIM_Cubic, leaveTangent=3.4262, time=1, value=1}
         ]}}}
      CurveIndex=&NormalizeDistanceRange(
       Start Position:NiagaraPosition=&SimulationPosition()
       End Position:NiagaraPosition=&VectorFromFloat(Value=&Multiply_Float(A=$Emitter.Scalability.DistanceFraction, B=15000))
       Distance:NiagaraFloat=15000
       Invert Normalized Range:NiagaraBool=false
      )
     )
    )
    Index 0 Param 2=1
    Index 0 Param 3=1
    Index 0 Param 4=1
    Module.Param0WriteEnabled=true
    Module.Param1WriteEnabled=false
    Module.Param2WriteEnabled=false
    Module.Param3WriteEnabled=false
   ) {module=DynamicMaterialParameters001}
  ]
  rendererClasses [&NiagaraMeshRendererProperties]
  renderers [{rendererClass=&NiagaraMeshRendererProperties, rendererIndex=0}]
 }}
summary {
 systemName=NS_Skyfire_LootOrb_Pos_Indicator
 emitters [{emitterName=PickupBeam, rendererClasses [&NiagaraMeshRendererProperties]}]
 userVariables [User.bIsBeingInteractedWith {type=NiagaraBool}]
}
